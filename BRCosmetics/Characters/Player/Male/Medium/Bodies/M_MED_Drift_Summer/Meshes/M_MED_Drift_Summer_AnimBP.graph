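# Secondary-motion AnimBP for the Drift Summer male body cosmetic, copying the attached parent mesh's pose and animating clothing and accessory bones.
# AnimDynamics nodes add spring-driven sway to necklaces, shorts, drawstrings and side strings, some with spherical limits.
# Additive bone transforms act as correctives for shorts, shirt and hips, weighted by thigh, calf, jumping and sprinting alphas.
# Weighted multi-bone constraints and a Look At anchor and aim the strings.
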AnimBlueprint /BRCosmetics/Characters/Player/Male/Medium/Bodies/M_MED_Drift_Summer/Meshes/M_MED_Drift_Summer_AnimBP
target_skeleton=&/M_MED_Drift_Summer_Skeleton
graphs [
 AnimGraph {nodes [
   "Transform (Modify) Bone - Bone: dyn_shorts_bk_02_r"(Alpha=Get calf_rot_backExtreme_r() -> (calf_rot_backExtreme_r:double), Translation="-10,0,0", Rotation="0,0,0", Scale="1,1,1") {
    rotationMode=BMM_Additive
    rotationSpace=BCS_BoneSpace
    scaleMode=BMM_Additive
    scaleSpace=BCS_BoneSpace
    translationMode=BMM_Additive
    translationSpace=BCS_ParentBoneSpace
    alphaBoolBlend {blendOption=Linear}
    alphaScaleBiasClamp {interpSpeedDecreasing=25, interpSpeedIncreasing=25}
    boneToModify {boneName=dyn_shorts_bk_02_r}
    translation {x=-10}
   }
   n1 = Get thigh_rot_r()
   n2 = Get calf_rot_back_r() -> (calf_rot_back_r:double)
   n3 = Get Shorts_Alpha() -> (Shorts_Alpha:double)
   n4 = Anim Dynamics - Bone: dyn_pendant_necklace_01!(ComponentPose=Local To Component(LocalPose=Copy Pose From Mesh() {bUseAttachedParent=true}), ExternalForce="0,0,0") -> (Pose) {
    alpha=0.2
    angularDampingOverride=0.999
    angularSpringConstant=500
    bAngularSpring=true
    bGravityOverrideInSimSpace=true
    bOverrideAngularDamping=true
    gravityScale=0.3
    numSolverIterationsPostUpdate=4
    numSolverIterationsPreUpdate=16
    simulationSpace=RootRelative
    alphaBoolBlend {blendOption=Linear}
    boundBone {boneName=dyn_pendant_necklace_01}
    componentLinearAccScale {z=1}
    componentLinearVelScale {z=1}
    physicsBodyDefinitions [{
      sphereCollisionRadius=0
      boundBone {boneName=dyn_pendant_necklace_01}
      boxExtents=<2 8 8>
      constraintSetup {angularTargetAxis=AxisY, angularLimitsMax=<0 30 60>, angularLimitsMin=<0 -30 0>, angularTarget=<0 1 0>}
      localJointOffset=<10 0 0>
     }]
   }
   n5 = Anim Dynamics - Bone: dyn_pendant_necklace_02!(ComponentPose=@n4, ExternalForce="0,0,0") -> (Pose) {
    alpha=0.2
    angularDampingOverride=0.999
    angularSpringConstant=500
    bAngularSpring=true
    bGravityOverrideInSimSpace=true
    bOverrideAngularDamping=true
    gravityScale=0.3
    numSolverIterationsPostUpdate=4
    numSolverIterationsPreUpdate=16
    simulationSpace=RootRelative
    alphaBoolBlend {blendOption=Linear}
    boundBone {boneName=dyn_pendant_necklace_02}
    componentLinearAccScale {z=1}
    componentLinearVelScale {z=1}
    physicsBodyDefinitions [{
      sphereCollisionRadius=0
      boundBone {boneName=dyn_pendant_necklace_02}
      boxExtents=<2 8 8>
      constraintSetup {angularTargetAxis=AxisY, angularLimitsMax=<0 30 60>, angularLimitsMin=<0 -30 0>, angularTarget=<0 1 0>}
      localJointOffset=<10 0 0>
     }]
   }
   n6 = Anim Dynamics - Bone: dyn_pendant_necklace_03!(ComponentPose=@n5, ExternalForce="0,0,0") -> (Pose) {
    alpha=0.15
    angularDampingOverride=0.999
    angularSpringConstant=500
    bAngularSpring=true
    bGravityOverrideInSimSpace=true
    bOverrideAngularDamping=true
    gravityScale=0.3
    numSolverIterationsPostUpdate=4
    numSolverIterationsPreUpdate=16
    simulationSpace=RootRelative
    alphaBoolBlend {blendOption=Linear}
    boundBone {boneName=dyn_pendant_necklace_03}
    componentLinearAccScale {z=1}
    componentLinearVelScale {z=1}
    physicsBodyDefinitions [{
      sphereCollisionRadius=0
      boundBone {boneName=dyn_pendant_necklace_03}
      boxExtents=<2 8 8>
      constraintSetup {angularTargetAxis=AxisY, angularLimitsMax=<0 30 60>, angularLimitsMin=<0 -30 0>, angularTarget=<0 1 0>}
      localJointOffset=<10 0 0>
     }]
   }
   n7 = Anim Dynamics - Bone: dyn_flower_necklace_01!(ComponentPose=@n6, ExternalForce="0,0,0") -> (Pose) {
    alpha=0.2
    angularDampingOverride=0.999
    bAngularSpring=true
    bGravityOverrideInSimSpace=true
    bOverrideAngularDamping=true
    gravityScale=0.1
    numSolverIterationsPostUpdate=4
    numSolverIterationsPreUpdate=16
    simulationSpace=RootRelative
    alphaBoolBlend {blendOption=Linear}
    boundBone {boneName=dyn_flower_necklace_01}
    componentLinearAccScale {z=0.5}
    componentLinearVelScale {z=0.5}
    physicsBodyDefinitions [{
      sphereCollisionRadius=0
      boundBone {boneName=dyn_flower_necklace_01}
      boxExtents=<5 10 10>
      constraintSetup {angularTargetAxis=AxisY, angularLimitsMax=<0 40 60>, angularLimitsMin=<0 -40 0>, angularTarget=<0 1 0>}
      localJointOffset=<5 0 0>
     }]
   }
   n8 = Anim Dynamics - Bone: dyn_flower_necklace_02!(ComponentPose=@n7, ExternalForce="0,0,0") -> (Pose) {
    alpha=0.35
    angularDampingOverride=0.999
    bAngularSpring=true
    bGravityOverrideInSimSpace=true
    bOverrideAngularDamping=true
    gravityScale=0.1
    numSolverIterationsPostUpdate=4
    numSolverIterationsPreUpdate=16
    simulationSpace=RootRelative
    alphaBoolBlend {blendOption=Linear}
    boundBone {boneName=dyn_flower_necklace_02}
    componentLinearAccScale {z=0.5}
    componentLinearVelScale {z=0.5}
    physicsBodyDefinitions [{
      sphereCollisionRadius=0
      boundBone {boneName=dyn_flower_necklace_02}
      boxExtents=<5 10 10>
      constraintSetup {angularTargetAxis=AxisY, angularLimitsMax=<0 40 60>, angularLimitsMin=<0 -40 0>, angularTarget=<0 1 0>}
      localJointOffset=<5 0 0>
     }]
   }
   n9 = Anim Dynamics - Bone: dyn_shorts_in_02_l(ComponentPose=@n8, Alpha=@n3, ExternalForce="0,0,0") {
    angularDampingOverride=0.999
    angularSpringConstant=100
    bAngularSpring=true
    bGravityOverrideInSimSpace=true
    bOverrideAngularDamping=true
    numSolverIterationsPostUpdate=4
    numSolverIterationsPreUpdate=16
    simulationSpace=RootRelative
    alphaBoolBlend {blendOption=Linear}
    boundBone {boneName=dyn_shorts_in_02_l}
    chainEnd {boneName=dyn_shorts_in_02_l}
    physicsBodyDefinitions [{
      sphereCollisionRadius=0
      boundBone {boneName=dyn_shorts_in_02_l}
      boxExtents=<25 30 12>
      constraintSetup {angularTargetAxis=AxisZ, twistAxis=AxisZ, angularLimitsMax=<10 10 5>, angularLimitsMin=<-10 -10 -5>, angularTarget=<0 0 1>}
      localJointOffset=<0 0 -3.5>
     }]
   }
   n10 = Anim Dynamics - Bone: dyn_shorts_out_02_l(ComponentPose=@n9, Alpha=@n3, ExternalForce="0,0,0") {
    angularDampingOverride=0.999
    angularSpringConstant=100
    bAngularSpring=true
    bGravityOverrideInSimSpace=true
    bOverrideAngularDamping=true
    numSolverIterationsPostUpdate=4
    numSolverIterationsPreUpdate=16
    simulationSpace=RootRelative
    alphaBoolBlend {blendOption=Linear}
    boundBone {boneName=dyn_shorts_out_02_l}
    chainEnd {boneName=dyn_shorts_out_02_l}
    physicsBodyDefinitions [{
      sphereCollisionRadius=0
      boundBone {boneName=dyn_shorts_out_02_l}
      boxExtents=<50 50 30>
      constraintSetup {angularTargetAxis=AxisZ, twistAxis=AxisZ, angularLimitsMax=<10 10 5>, angularLimitsMin=<-10 -10 -5>, angularTarget=<0 0 1>}
      localJointOffset=<0 0 -6.5>
     }]
   }
   n11 = Anim Dynamics - Bone: dyn_shorts_fr_02_l(ComponentPose=@n10, Alpha=@n3, ExternalForce="0,0,0") {
    angularDampingOverride=0.999
    angularSpringConstant=300
    bAngularSpring=true
    bGravityOverrideInSimSpace=true
    bOverrideAngularDamping=true
    numSolverIterationsPostUpdate=4
    numSolverIterationsPreUpdate=16
    simulationSpace=RootRelative
    alphaBoolBlend {blendOption=Linear}
    boundBone {boneName=dyn_shorts_fr_02_l}
    chainEnd {boneName=dyn_shorts_fr_02_l}
    physicsBodyDefinitions [{
      sphereCollisionRadius=0
      boundBone {boneName=dyn_shorts_fr_02_l}
      boxExtents=<18 25 8>
      constraintSetup {angularTargetAxis=AxisZ, twistAxis=AxisZ, angularLimitsMax=<10 0 5>, angularLimitsMin=<-10 -20 -5>, angularTarget=<0 0 1>}
      localJointOffset=<0 0 -5>
     }]
   }
   n12 = Anim Dynamics - Bone: dyn_shorts_bk_02_l(ComponentPose=@n11, Alpha=@n3, ExternalForce="0,0,0") {
    angularDampingOverride=0.999
    angularSpringConstant=300
    bAngularSpring=true
    bGravityOverrideInSimSpace=true
    bOverrideAngularDamping=true
    numSolverIterationsPostUpdate=4
    numSolverIterationsPreUpdate=16
    simulationSpace=RootRelative
    alphaBoolBlend {blendOption=Linear}
    boundBone {boneName=dyn_shorts_bk_02_l}
    chainEnd {boneName=dyn_shorts_bk_02_l}
    physicsBodyDefinitions [{
      sphereCollisionRadius=0
      boundBone {boneName=dyn_shorts_bk_02_l}
      boxExtents=<40 38 15>
      constraintSetup {angularTargetAxis=AxisZ, twistAxis=AxisZ, angularLimitsMax=<0 0 5>, angularLimitsMin=<0 -20 -5>, angularTarget=<0 0 1>}
      localJointOffset=<0 0 -3.5>
     }]
   }
   n13 = Anim Dynamics - Bone: dyn_shorts_in_02_r(ComponentPose=@n12, Alpha=@n3, ExternalForce="0,0,0") {
    angularDampingOverride=0.999
    angularSpringConstant=100
    bAngularSpring=true
    bGravityOverrideInSimSpace=true
    bOverrideAngularDamping=true
    numSolverIterationsPostUpdate=4
    numSolverIterationsPreUpdate=16
    simulationSpace=RootRelative
    alphaBoolBlend {blendOption=Linear}
    boundBone {boneName=dyn_shorts_in_02_r}
    chainEnd {boneName=dyn_shorts_in_02_r}
    componentLinearAccScale=<1 1 1>
    componentLinearVelScale=<1 1 1>
    physicsBodyDefinitions [{
      sphereCollisionRadius=0
      boundBone {boneName=dyn_shorts_in_02_r}
      boxExtents=<25 29 12>
      constraintSetup {angularTargetAxis=AxisZ, twistAxis=AxisZ, angularLimitsMax=<10 0 5>, angularLimitsMin=<-10 -10 -5>, angularTarget=<0 0 1>}
      localJointOffset=<0 0 3.5>
     }]
   }
   n14 = Anim Dynamics - Bone: dyn_shorts_out_02_r(ComponentPose=@n13, Alpha=@n3, ExternalForce="0,0,0") {
    angularDampingOverride=0.999
    angularSpringConstant=100
    bAngularSpring=true
    bGravityOverrideInSimSpace=true
    bOverrideAngularDamping=true
    numSolverIterationsPostUpdate=4
    numSolverIterationsPreUpdate=16
    simulationSpace=RootRelative
    alphaBoolBlend {blendOption=Linear}
    boundBone {boneName=dyn_shorts_out_02_r}
    chainEnd {boneName=dyn_shorts_out_02_r}
    componentLinearAccScale=<1 1 1>
    componentLinearVelScale=<1 1 1>
    physicsBodyDefinitions [{
      sphereCollisionRadius=0
      boundBone {boneName=dyn_shorts_out_02_r}
      boxExtents=<50 50 30>
      constraintSetup {angularTargetAxis=AxisZ, twistAxis=AxisZ, angularLimitsMax=<10 10 5>, angularLimitsMin=<-10 -10 -5>, angularTarget=<0 0 1>}
      localJointOffset=<0 0 6.5>
     }]
   }
   n15 = Anim Dynamics - Bone: dyn_shorts_fr_02_r(ComponentPose=@n14, Alpha=@n3, ExternalForce="0,0,0") {
    angularDampingOverride=0.999
    angularSpringConstant=300
    bAngularSpring=true
    bGravityOverrideInSimSpace=true
    bOverrideAngularDamping=true
    numSolverIterationsPostUpdate=4
    numSolverIterationsPreUpdate=16
    simulationSpace=RootRelative
    alphaBoolBlend {blendOption=Linear}
    boundBone {boneName=dyn_shorts_fr_02_r}
    chainEnd {boneName=dyn_shorts_fr_02_r}
    componentLinearAccScale=<1 1 1>
    componentLinearVelScale=<1 1 1>
    physicsBodyDefinitions [{
      sphereCollisionRadius=0
      boundBone {boneName=dyn_shorts_fr_02_r}
      boxExtents=<26 22 8>
      constraintSetup {angularTargetAxis=AxisZ, twistAxis=AxisZ, angularLimitsMax=<10 0 5>, angularLimitsMin=<-10 -20 -5>, angularTarget=<0 0 1>}
      localJointOffset=<0 0 5>
     }]
   }
   n16 = Anim Dynamics - Bone: dyn_shorts_bk_02_r(ComponentPose=@n15, Alpha=@n3, ExternalForce="0,0,0") {
    angularDampingOverride=0.999
    angularSpringConstant=300
    bAngularSpring=true
    bGravityOverrideInSimSpace=true
    bOverrideAngularDamping=true
    numSolverIterationsPostUpdate=4
    numSolverIterationsPreUpdate=16
    simulationSpace=RootRelative
    alphaBoolBlend {blendOption=Linear}
    boundBone {boneName=dyn_shorts_bk_02_r}
    chainEnd {boneName=dyn_shorts_bk_02_r}
    componentLinearAccScale=<1 1 1>
    componentLinearVelScale=<1 1 1>
    physicsBodyDefinitions [{
      sphereCollisionRadius=0
      boundBone {boneName=dyn_shorts_bk_02_r}
      boxExtents=<40 36 15>
      constraintSetup {angularTargetAxis=AxisZ, twistAxis=AxisZ, angularLimitsMax=<0 0 5>, angularLimitsMin=<0 -20 -5>, angularTarget=<0 0 1>}
      localJointOffset=<0 0 3.5>
     }]
   }
   n17 = Get calf_rot_back_l() -> (calf_rot_back_l:double)
   n18 = "Transform (Modify) Bone - Bone: dyn_hipThigh_l"(ComponentPose=@n16, Alpha=@n17, Translation="-6,0,0", Rotation="0,0,0", Scale="1,1,1") {
    rotationMode=BMM_Additive
    rotationSpace=BCS_BoneSpace
    scaleMode=BMM_Additive
    scaleSpace=BCS_BoneSpace
    translationMode=BMM_Additive
    translationSpace=BCS_ParentBoneSpace
    alphaBoolBlend {blendOption=Linear}
    alphaScaleBiasClamp {interpSpeedDecreasing=25, interpSpeedIncreasing=25}
    boneToModify {boneName=dyn_hipThigh_l}
    translation {x=-6}
   }
   n19 = "Transform (Modify) Bone - Bone: dyn_buttThigh_l"(ComponentPose=@n18, Alpha=@n17, Translation="15,-15,0", Rotation="0,0,0", Scale="1,1,1") {
    rotationMode=BMM_Additive
    rotationSpace=BCS_BoneSpace
    scaleMode=BMM_Additive
    scaleSpace=BCS_BoneSpace
    translationMode=BMM_Additive
    translationSpace=BCS_ParentBoneSpace
    alphaBoolBlend {blendOption=Linear}
    alphaScaleBiasClamp {interpSpeedDecreasing=25, interpSpeedIncreasing=25}
    boneToModify {boneName=dyn_buttThigh_l}
    translation=<15 -15>
   }
   n20 = "Transform (Modify) Bone - Bone: dyn_shorts_main_l"(ComponentPose=@n19, Alpha=@n17, Translation="0,-4,0", Rotation="0,0,-5", Scale="1,1,1") {
    rotationMode=BMM_Additive
    rotationSpace=BCS_BoneSpace
    scaleMode=BMM_Additive
    scaleSpace=BCS_BoneSpace
    translationMode=BMM_Additive
    translationSpace=BCS_ParentBoneSpace
    alphaBoolBlend {blendOption=Linear}
    alphaScaleBiasClamp {interpSpeedDecreasing=25, interpSpeedIncreasing=25}
    boneToModify {boneName=dyn_shorts_main_l}
    rotation {roll=-5}
    translation {y=-4}
   }
   n21 = "Transform (Modify) Bone - Bone: dyn_shorts_out_l"(ComponentPose=@n20, Alpha=@n17, Translation="0,0,0", Rotation="0,0,20", Scale="1,1,1") {
    rotationMode=BMM_Additive
    rotationSpace=BCS_BoneSpace
    scaleMode=BMM_Additive
    scaleSpace=BCS_BoneSpace
    translationMode=BMM_Additive
    translationSpace=BCS_ParentBoneSpace
    alphaBoolBlend {blendOption=Linear}
    alphaScaleBiasClamp {interpSpeedDecreasing=25, interpSpeedIncreasing=25}
    boneToModify {boneName=dyn_shorts_out_l}
    rotation {roll=20}
    translation {y=0, z=0}
   }
   n22 = "Transform (Modify) Bone - Bone: dyn_shorts_in_l"(ComponentPose=@n21, Alpha=@n17, Translation="0,0,0", Rotation="0,0,-20", Scale="1,1,1") {
    rotationMode=BMM_Additive
    rotationSpace=BCS_BoneSpace
    scaleMode=BMM_Additive
    scaleSpace=BCS_BoneSpace
    translationMode=BMM_Additive
    translationSpace=BCS_ParentBoneSpace
    alphaBoolBlend {blendOption=Linear}
    alphaScaleBiasClamp {interpSpeedDecreasing=25, interpSpeedIncreasing=25}
    boneToModify {boneName=dyn_shorts_in_l}
    rotation {roll=-20}
   }
   n23 = "Transform (Modify) Bone - Bone: dyn_shorts_fr_l"(ComponentPose=@n22, Alpha=@n17, Translation="0,0,0", Rotation="10,0,-10", Scale="1,1,1") {
    rotationMode=BMM_Additive
    rotationSpace=BCS_BoneSpace
    scaleMode=BMM_Additive
    scaleSpace=BCS_BoneSpace
    translationMode=BMM_Additive
    translationSpace=BCS_ParentBoneSpace
    alphaBoolBlend {blendOption=Linear}
    alphaScaleBiasClamp {interpSpeedDecreasing=25, interpSpeedIncreasing=25}
    boneToModify {boneName=dyn_shorts_fr_l}
    rotation=<rot 10 0 -10>
   }
   n24 = "Transform (Modify) Bone - Bone: dyn_shorts_bk_02_l"(ComponentPose=@n23, Alpha=@n17, Translation="2,0,5", Rotation="120,-10,0", Scale="1,1,0.4") {
    rotationMode=BMM_Additive
    rotationSpace=BCS_BoneSpace
    scaleMode=BMM_Additive
    scaleSpace=BCS_BoneSpace
    translationMode=BMM_Additive
    translationSpace=BCS_ParentBoneSpace
    alphaBoolBlend {blendOption=Linear}
    alphaScaleBiasClamp {interpSpeedDecreasing=25, interpSpeedIncreasing=25}
    boneToModify {boneName=dyn_shorts_bk_02_l}
    rotation {pitch=120, yaw=-10}
    scale {z=0.4}
    translation {x=2, z=5}
   }
   n25 = "Transform (Modify) Bone - Bone: dyn_shorts_bk_02_l"(
    ComponentPose=@n24
    Alpha=Get calf_rot_backExtreme_l() -> (calf_rot_backExtreme_l:double)
    Translation="0,0,10"
    Rotation="0,0,0"
    Scale="1,1,1"
   ) {
    rotationMode=BMM_Additive
    rotationSpace=BCS_BoneSpace
    scaleMode=BMM_Additive
    scaleSpace=BCS_BoneSpace
    translationMode=BMM_Additive
    translationSpace=BCS_ParentBoneSpace
    alphaBoolBlend {blendOption=Linear}
    alphaScaleBiasClamp {bInterpResult=true, interpSpeedDecreasing=25, interpSpeedIncreasing=25}
    boneToModify {boneName=dyn_shorts_bk_02_l}
    translation {z=10}
   }
   n26 = "Transform (Modify) Bone - Bone: dyn_shorts_out_02_l"(ComponentPose=@n25, Alpha=@n17, Translation="0,0,0", Rotation="-20,-20,40", Scale="1,1,0.7") {
    rotationMode=BMM_Additive
    rotationSpace=BCS_BoneSpace
    scaleMode=BMM_Additive
    scaleSpace=BCS_BoneSpace
    translationMode=BMM_Additive
    translationSpace=BCS_ParentBoneSpace
    alphaBoolBlend {blendOption=Linear}
    alphaScaleBiasClamp {interpSpeedDecreasing=25, interpSpeedIncreasing=25}
    boneToModify {boneName=dyn_shorts_out_02_l}
    rotation=<rot -20 -20 40>
    scale {z=0.7}
   }
   n27 = "Transform (Modify) Bone - Bone: dyn_shorts_in_02_l"(ComponentPose=@n26, Alpha=@n17, Translation="0,0,3", Rotation="0,0,-50", Scale="1,1,0.7") {
    rotationMode=BMM_Additive
    rotationSpace=BCS_BoneSpace
    scaleMode=BMM_Additive
    scaleSpace=BCS_BoneSpace
    translationMode=BMM_Additive
    translationSpace=BCS_ParentBoneSpace
    alphaBoolBlend {blendOption=Linear}
    alphaScaleBiasClamp {interpSpeedDecreasing=25, interpSpeedIncreasing=25}
    boneToModify {boneName=dyn_shorts_in_02_l}
    rotation {roll=-50}
    scale {z=0.7}
    translation {z=3}
   }
   n28 = "Transform (Modify) Bone - Bone: dyn_shorts_fr_02_l"(ComponentPose=@n27, Alpha=@n17, Translation="0,0,0", Rotation="-40,0,-10", Scale="1,1,1") {
    rotationMode=BMM_Additive
    rotationSpace=BCS_BoneSpace
    scaleMode=BMM_Additive
    scaleSpace=BCS_BoneSpace
    translationMode=BMM_Additive
    translationSpace=BCS_ParentBoneSpace
    alphaBoolBlend {blendOption=Linear}
    alphaScaleBiasClamp {interpSpeedDecreasing=25, interpSpeedIncreasing=25}
    boneToModify {boneName=dyn_shorts_fr_02_l}
    rotation {pitch=-40, roll=-10}
    translation {x=0, z=0}
   }
   n29 = "Transform (Modify) Bone - Bone: dyn_shorts_main_r"(ComponentPose=@n28, Alpha=@n2, Translation="0,0,0", Rotation="0,0,10", Scale="1,1,1") {
    rotationMode=BMM_Additive
    rotationSpace=BCS_BoneSpace
    scaleMode=BMM_Additive
    scaleSpace=BCS_BoneSpace
    translationMode=BMM_Additive
    translationSpace=BCS_ParentBoneSpace
    alphaBoolBlend {blendOption=Linear}
    alphaScaleBiasClamp {interpSpeedDecreasing=25, interpSpeedIncreasing=25}
    boneToModify {boneName=dyn_shorts_main_r}
    rotation {roll=10}
   }
   n30 = "Transform (Modify) Bone - Bone: dyn_shorts_bk_02_r"(ComponentPose=@n29, Alpha=@n2, Translation="0,0.000,-3.5", Rotation="100,0,0", Scale="1,1,0.5") {
    rotationMode=BMM_Additive
    rotationSpace=BCS_BoneSpace
    scaleMode=BMM_Additive
    scaleSpace=BCS_BoneSpace
    translationMode=BMM_Additive
    translationSpace=BCS_BoneSpace
    alphaBoolBlend {blendOption=Linear}
    alphaScaleBiasClamp {interpSpeedDecreasing=25, interpSpeedIncreasing=25}
    boneToModify {boneName=dyn_shorts_bk_02_r}
    rotation {pitch=100}
    scale {z=0.5}
    translation {z=-3.5}
   }
   n31 = "Transform (Modify) Bone - Bone: dyn_shorts_out_r"(ComponentPose=@n30, Alpha=@n2, Translation="0,0,0", Rotation="0,0,15", Scale="1,1,1") {
    rotationMode=BMM_Additive
    rotationSpace=BCS_BoneSpace
    scaleMode=BMM_Additive
    scaleSpace=BCS_BoneSpace
    translationMode=BMM_Additive
    translationSpace=BCS_BoneSpace
    alphaBoolBlend {blendOption=Linear}
    alphaScaleBiasClamp {interpSpeedDecreasing=25, interpSpeedIncreasing=25}
    boneToModify {boneName=dyn_shorts_out_r}
    rotation {roll=15}
   }
   n32 = "Transform (Modify) Bone - Bone: dyn_shorts_out_02_r"(ComponentPose=@n31, Alpha=@n2, Translation="0,0,-3", Rotation="15,0,0", Scale="1,1,1") {
    rotationMode=BMM_Additive
    rotationSpace=BCS_BoneSpace
    scaleMode=BMM_Additive
    scaleSpace=BCS_BoneSpace
    translationMode=BMM_Additive
    translationSpace=BCS_BoneSpace
    alphaBoolBlend {blendOption=Linear}
    alphaScaleBiasClamp {interpSpeedDecreasing=25, interpSpeedIncreasing=25}
    boneToModify {boneName=dyn_shorts_out_02_r}
    rotation {pitch=15}
    translation {z=-3}
   }
   n33 = "Transform (Modify) Bone - Bone: dyn_shorts_in_02_r"(ComponentPose=@n32, Alpha=@n2, Translation="0,0,0", Rotation="0,0,-40", Scale="1,1,1") {
    rotationMode=BMM_Additive
    rotationSpace=BCS_BoneSpace
    scaleMode=BMM_Additive
    scaleSpace=BCS_BoneSpace
    translationMode=BMM_Additive
    translationSpace=BCS_BoneSpace
    alphaBoolBlend {blendOption=Linear}
    alphaScaleBiasClamp {interpSpeedDecreasing=25, interpSpeedIncreasing=25}
    boneToModify {boneName=dyn_shorts_in_02_r}
    rotation {roll=-40}
   }
   n34 = "Transform (Modify) Bone - Bone: dyn_shorts_fr_02_r"(ComponentPose=@n33, Alpha=@n2, Translation="-2,0,0", Rotation="-15,0,0", Scale="1,1,1") {
    rotationMode=BMM_Additive
    rotationSpace=BCS_BoneSpace
    scaleMode=BMM_Additive
    scaleSpace=BCS_BoneSpace
    translationMode=BMM_Additive
    translationSpace=BCS_BoneSpace
    alphaBoolBlend {blendOption=Linear}
    alphaScaleBiasClamp {interpSpeedDecreasing=25, interpSpeedIncreasing=25}
    boneToModify {boneName=dyn_shorts_fr_02_r}
    rotation {pitch=-15}
    translation {x=-2}
   }
   n35 = Get jumping_alpha() -> (jumping_alpha:double)
   n36 = "Transform (Modify) Bone - Bone: dyn_shorts_bk_l"(ComponentPose=@n34, Alpha=@n35, Translation="-2.5,-5,2.5", Rotation="0,0,0", Scale="1,1,1") {
    rotationMode=BMM_Additive
    rotationSpace=BCS_BoneSpace
    scaleMode=BMM_Additive
    scaleSpace=BCS_BoneSpace
    translationMode=BMM_Additive
    translationSpace=BCS_ParentBoneSpace
    alphaBoolBlend {blendOption=Linear}
    alphaScaleBiasClamp {bInterpResult=true, interpSpeedDecreasing=2, interpSpeedIncreasing=2}
    boneToModify {boneName=dyn_shorts_bk_l}
    translation=<-2.5 -5 2.5>
   }
   n37 = "Transform (Modify) Bone - Bone: dyn_shorts_bk_r"(ComponentPose=@n36, Alpha=@n35, Translation="-2.5,5,-10", Rotation="0,0,0", Scale="1,1,1") {
    rotationMode=BMM_Additive
    rotationSpace=BCS_BoneSpace
    scaleMode=BMM_Additive
    scaleSpace=BCS_BoneSpace
    translationMode=BMM_Additive
    translationSpace=BCS_ParentBoneSpace
    alphaBoolBlend {blendOption=Linear}
    alphaScaleBiasClamp {bInterpResult=true, interpSpeedDecreasing=2, interpSpeedIncreasing=2}
    boneToModify {boneName=dyn_shorts_bk_r}
    translation=<-2.5 5 -10>
   }
   n38 = Get sprinting_alpha() -> (sprinting_alpha:double)
   n39 = "Transform (Modify) Bone - Bone: dyn_shorts_bk_r"(ComponentPose=@n37, Alpha=@n38, Translation="-3,3,-3", Rotation="0,0,0", Scale="1,1,1") {
    rotationMode=BMM_Additive
    rotationSpace=BCS_BoneSpace
    scaleMode=BMM_Additive
    scaleSpace=BCS_BoneSpace
    translationMode=BMM_Additive
    translationSpace=BCS_ParentBoneSpace
    alphaBoolBlend {blendOption=Linear}
    alphaScaleBiasClamp {bInterpResult=true}
    boneToModify {boneName=dyn_shorts_bk_r}
    translation=<-3 3 -3>
   }
   n40 = "Transform (Modify) Bone - Bone: dyn_shorts_in_r"(ComponentPose=@n39, Alpha=@n38, Translation="-3,0,-3", Rotation="0,0,0", Scale="1,1,1") {
    rotationMode=BMM_Additive
    rotationSpace=BCS_BoneSpace
    scaleMode=BMM_Additive
    scaleSpace=BCS_BoneSpace
    translationMode=BMM_Additive
    translationSpace=BCS_ParentBoneSpace
    alphaBoolBlend {blendOption=Linear}
    alphaScaleBiasClamp {bInterpResult=true}
    boneToModify {boneName=dyn_shorts_in_r}
    translation {x=-3, z=-3}
   }
   n41 = Get thigh_rot_l()
   n42 = "Transform (Modify) Bone - Bone: dyn_shirtFront_l"(ComponentPose=@n40, Alpha=@n41, Translation="-7,-3,0", Rotation="0,0,0", Scale="1,1,1") {
    rotationMode=BMM_Additive
    rotationSpace=BCS_BoneSpace
    scaleMode=BMM_Additive
    scaleSpace=BCS_BoneSpace
    translationMode=BMM_Additive
    translationSpace=BCS_ParentBoneSpace
    alphaBoolBlend {blendOption=Linear}
    alphaScaleBiasClamp {interpSpeedDecreasing=25, interpSpeedIncreasing=25, scale=0.6}
    boneToModify {boneName=dyn_shirtFront_l}
    translation=<-7 -3>
   }
   n43 = "Transform (Modify) Bone - Bone: dyn_shirtMid_l"(ComponentPose=@n42, Alpha=@n41, Translation="-20,0,0", Rotation="0,0,0", Scale="0.8,1,1") {
    rotationMode=BMM_Additive
    rotationSpace=BCS_ParentBoneSpace
    scaleMode=BMM_Additive
    scaleSpace=BCS_BoneSpace
    translationMode=BMM_Additive
    translationSpace=BCS_BoneSpace
    alphaBoolBlend {blendOption=Linear}
    alphaScaleBiasClamp {interpSpeedDecreasing=25, interpSpeedIncreasing=25}
    boneToModify {boneName=dyn_shirtMid_l}
    scale {x=0.8}
    translation {x=-20}
   }
   n44 = "Transform (Modify) Bone - Bone: dyn_shirtSide_l"(ComponentPose=@n43, Alpha=@n41, Translation="0,0,0", Rotation="0,0,0", Scale="1,1,1") {
    rotationMode=BMM_Additive
    rotationSpace=BCS_BoneSpace
    scaleMode=BMM_Additive
    scaleSpace=BCS_BoneSpace
    translationMode=BMM_Additive
    translationSpace=BCS_ParentBoneSpace
    alphaBoolBlend {blendOption=Linear}
    alphaScaleBiasClamp {interpSpeedDecreasing=25, interpSpeedIncreasing=25}
    boneToModify {boneName=dyn_shirtSide_l}
   }
   n45 = "Transform (Modify) Bone - Bone: dyn_shirtFront_r"(ComponentPose=@n44, Alpha=@n41, Translation="0,0,0", Rotation="0,-15,0", Scale="1,1,1") {
    rotationMode=BMM_Additive
    rotationSpace=BCS_BoneSpace
    scaleMode=BMM_Additive
    scaleSpace=BCS_BoneSpace
    translationMode=BMM_Additive
    translationSpace=BCS_BoneSpace
    alphaBoolBlend {blendOption=Linear}
    alphaScaleBiasClamp {interpSpeedDecreasing=25, interpSpeedIncreasing=25}
    boneToModify {boneName=dyn_shirtFront_r}
    rotation {yaw=-15}
   }
   n46 = "Transform (Modify) Bone - Bone: dyn_hipThigh_l"(ComponentPose=@n45, Alpha=@n41, Translation="0,-2,0", Rotation="0,0,0", Scale="1,1,1") {
    rotationMode=BMM_Additive
    rotationSpace=BCS_BoneSpace
    scaleMode=BMM_Additive
    scaleSpace=BCS_BoneSpace
    translationMode=BMM_Additive
    translationSpace=BCS_BoneSpace
    alphaBoolBlend {blendOption=Linear}
    alphaScaleBiasClamp {interpSpeedDecreasing=25, interpSpeedIncreasing=25}
    boneToModify {boneName=dyn_hipThigh_l}
    translation {y=-2}
   }
   n47 = "Transform (Modify) Bone - Bone: dyn_shirtFront_r"(ComponentPose=@n46, Alpha=@n1, Translation="-5,-5,0", Rotation="0,0,0", Scale="1,1,1") {
    rotationMode=BMM_Additive
    rotationSpace=BCS_BoneSpace
    scaleMode=BMM_Additive
    scaleSpace=BCS_BoneSpace
    translationMode=BMM_Additive
    translationSpace=BCS_BoneSpace
    alphaBoolBlend {blendOption=Linear}
    alphaScaleBiasClamp {interpSpeedDecreasing=25, interpSpeedIncreasing=25}
    boneToModify {boneName=dyn_shirtFront_r}
    translation=<-5 -5>
   }
   n48 = "Transform (Modify) Bone - Bone: dyn_shirtMid_r"(ComponentPose=@n47, Alpha=@n1, Translation="-20,0,5", Rotation="0,0,0", Scale="1,1,1") {
    rotationMode=BMM_Additive
    rotationSpace=BCS_BoneSpace
    scaleMode=BMM_Additive
    scaleSpace=BCS_BoneSpace
    translationMode=BMM_Additive
    translationSpace=BCS_BoneSpace
    alphaBoolBlend {blendOption=Linear}
    alphaScaleBiasClamp {interpSpeedDecreasing=25, interpSpeedIncreasing=25}
    boneToModify {boneName=dyn_shirtMid_r}
    translation {x=-20, z=5}
   }
   n49 = "Transform (Modify) Bone - Bone: dyn_shirtSide_r"(ComponentPose=@n48, Alpha=@n1, Translation="-10,0,0", Rotation="0,0,0", Scale="1,1,1") {
    rotationMode=BMM_Additive
    rotationSpace=BCS_BoneSpace
    scaleMode=BMM_Additive
    scaleSpace=BCS_BoneSpace
    translationMode=BMM_Additive
    translationSpace=BCS_BoneSpace
    alphaBoolBlend {blendOption=Linear}
    alphaScaleBiasClamp {interpSpeedDecreasing=25, interpSpeedIncreasing=25}
    boneToModify {boneName=dyn_shirtSide_r}
    translation {x=-10}
   }
   n50 = "Transform (Modify) Bone - Bone: dyn_pelvis_scale_r"(ComponentPose=@n49, Alpha=@n1, Translation="-10,0,0", Rotation="0,0,0", Scale="1,1,1") {
    rotationMode=BMM_Additive
    rotationSpace=BCS_BoneSpace
    scaleMode=BMM_Additive
    scaleSpace=BCS_BoneSpace
    translationMode=BMM_Additive
    translationSpace=BCS_BoneSpace
    alphaBoolBlend {blendOption=Linear}
    alphaScaleBiasClamp {interpSpeedDecreasing=25, interpSpeedIncreasing=25}
    boneToModify {boneName=dyn_pelvis_scale_r}
    translation {x=-10}
   }
   n51 = "Transform (Modify) Bone - Bone: dyn_shirtSide_r"(ComponentPose=@n50, Alpha=Get thigh_toSide_r() -> (thigh_toSide_r:double), Translation="-8,0,0", Rotation="0,0,0", Scale="1,1,1") {
    rotationMode=BMM_Additive
    rotationSpace=BCS_BoneSpace
    scaleMode=BMM_Additive
    scaleSpace=BCS_BoneSpace
    translationMode=BMM_Additive
    translationSpace=BCS_BoneSpace
    alphaBoolBlend {blendOption=Linear}
    alphaScaleBiasClamp {interpSpeedDecreasing=25, interpSpeedIncreasing=25}
    boneToModify {boneName=dyn_shirtSide_r}
    translation {x=-8}
   }
   n52 = "Constraint - dyn_drawstring_01_Base (PPPPPPPPP)"!(
    ComponentPose=@n51
    Alpha=1
    ConstraintWeights_0=0.0248
    ConstraintWeights_1:float=0.0693
    ConstraintWeights_2:float=0.0562
    ConstraintWeights_3:float=0.0053
    ConstraintWeights_4:float=0.0363
    ConstraintWeights_5:float=0.0464
    ConstraintWeights_6:float=0.0511
    ConstraintWeights_7:float=0.6945
    ConstraintWeights_8:float=0.0162
   ) -> (Pose) {
    alphaBoolBlend {blendOption=Linear}
    boneToModify {boneName=dyn_drawstring_01_Base}
    constraintSetup [
     {targetBone {boneName=dyn_hipPelvis_l}}
     {targetBone {boneName=thigh_twist_01_r}}
     {targetBone {boneName=spine_01}}
     {targetBone {boneName=spine_02}}
     {targetBone {boneName=thigh_twist_01_l}}
     {targetBone {boneName=dyn_pelvis_scale_r}}
     {targetBone {boneName=thigh_l}}
     {targetBone {boneName=pelvis}}
     {targetBone {boneName=dyn_thigh_scale_r}}
    ]
    constraintWeights [1,1,1,1,1,1,1,1,1]
   }
   n53 = "Constraint - thighFront_c (PPPPPPPPPPPPPP)"!(
    ComponentPose=@n52
    Alpha=1
    ConstraintWeights_0=0
    ConstraintWeights_1:float=0.0941
    ConstraintWeights_2:float=0.0455
    ConstraintWeights_3:float=0.0279
    ConstraintWeights_4:float=0.1906
    ConstraintWeights_5:float=0.072
    ConstraintWeights_6:float=0.038
    ConstraintWeights_7:float=0.1882
    ConstraintWeights_8:float=0.015
    ConstraintWeights_9:float=0.0125
    ConstraintWeights_10:float=0.0766
    ConstraintWeights_11:float=0.2104
    ConstraintWeights_12:float=0
    ConstraintWeights_13:float=0.0292
   ) -> (Pose) {
    alphaBoolBlend {blendOption=Linear}
    boneToModify {boneName=thighFront_c}
    constraintSetup [
     {targetBone {boneName=dyn_shorts_bk_l}}
     {targetBone {boneName=dyn_shorts_fr_l}}
     {targetBone {boneName=dyn_thighPelvis_l}}
     {targetBone {boneName=dyn_shorts_in_l}}
     {targetBone {boneName=thigh_twist_01_r}}
     {targetBone {boneName=dyn_shorts_main_l}}
     {targetBone {boneName=dyn_shorts_main_r}}
     {targetBone {boneName=thigh_twist_01_l}}
     {targetBone {boneName=dyn_thigh_scale_r}}
     {targetBone {boneName=thigh_l}}
     {targetBone {boneName=dyn_shorts_fr_r}}
     {targetBone {boneName=pelvis}}
     {targetBone {boneName=dyn_shorts_bk_r}}
     {targetBone {boneName=dyn_shorts_in_r}}
    ]
    constraintWeights [1,1,1,1,1,1,1,1,1,1,1,1,1,1]
   }
   n54 = Look At - Bone: dyn_drawstring_01(ComponentPose=@n53, Alpha=1) {
    alphaBoolBlend {blendOption=Linear}
    boneToModify {boneName=dyn_drawstring_01}
    lookAtLocation=<0 1>
    lookAtTarget {boneReference {boneName=thighFront_c}}
    lookAt_Axis {axis=<1 0>}
    lookUp_Axis {axis {x=0, z=1}}
   }
   n55 = "Transform (Modify) Bone - Bone: dyn_drawstring_01"(ComponentPose=@n54, Alpha=Get thigh_rot_both() -> (thigh_rot_both:double), Translation="0,0,0", Rotation="0,20,0", Scale="1,1,1") {
    rotationMode=BMM_Additive
    rotationSpace=BCS_BoneSpace
    scaleMode=BMM_Additive
    scaleSpace=BCS_BoneSpace
    translationMode=BMM_Additive
    translationSpace=BCS_BoneSpace
    alphaBoolBlend {blendOption=Linear}
    alphaScaleBiasClamp {interpSpeedDecreasing=25, interpSpeedIncreasing=25}
    boneToModify {boneName=dyn_drawstring_01}
    rotation {yaw=20}
   }
   n56 = Get thigh_rot_either() -> (thigh_rot_either:double)
   n57 = "Transform (Modify) Bone - Bone: dyn_drawstring_02_l"(ComponentPose=@n55, Alpha=@n56, Translation="0,0,0", Rotation="-20,0,0", Scale="1,1,1") {
    rotationMode=BMM_Additive
    rotationSpace=BCS_BoneSpace
    scaleMode=BMM_Additive
    scaleSpace=BCS_BoneSpace
    translationMode=BMM_Additive
    translationSpace=BCS_BoneSpace
    alphaBoolBlend {blendOption=Linear}
    alphaScaleBiasClamp {interpSpeedDecreasing=25, interpSpeedIncreasing=25}
    boneToModify {boneName=dyn_drawstring_02_l}
    rotation {pitch=-20}
   }
   n58 = "Transform (Modify) Bone - Bone: dyn_drawstring_02_r"(ComponentPose=@n57, Alpha=@n56, Translation="0,0,0", Rotation="-20,0,0", Scale="1,1,1") {
    rotationMode=BMM_Additive
    rotationSpace=BCS_BoneSpace
    scaleMode=BMM_Additive
    scaleSpace=BCS_BoneSpace
    translationMode=BMM_Additive
    translationSpace=BCS_BoneSpace
    alphaBoolBlend {blendOption=Linear}
    alphaScaleBiasClamp {interpSpeedDecreasing=25, interpSpeedIncreasing=25}
    boneToModify {boneName=dyn_drawstring_02_r}
    rotation {pitch=-20}
   }
   n59 = Anim Dynamics - Bone: dyn_drawstring_03_r(ComponentPose=@n58, Alpha=1, ExternalForce="0,0,0") {
    angularDampingOverride=1
    angularSpringConstant=100
    bAngularSpring=true
    bGravityOverrideInSimSpace=true
    bOverrideAngularDamping=true
    bUseGravityOverride=true
    bUsePlanarLimit=false
    gravityScale=0.3
    numSolverIterationsPostUpdate=4
    numSolverIterationsPreUpdate=16
    alphaBoolBlend {blendOption=Linear}
    boundBone {boneName=dyn_drawstring_03_r}
    chainEnd {boneName=dyn_drawstring_03_r}
    gravityOverride {z=-980}
    physicsBodyDefinitions [{
      collisionType=CustomSphere
      sphereCollisionRadius=1
      boundBone {boneName=dyn_drawstring_03_r}
      boxExtents=<5 10 10>
      constraintSetup {angularLimitsMax=<0 60 60>, angularLimitsMin=<0 -60 0>, angularTarget=<1 0 0>}
      localJointOffset=<8 0 0>
     }]
    sphericalLimits [{limitRadius=20, drivingBone {boneName=thighFront_c}, sphereLocalOffset=<0 -20 0>}]
   }
   n60 = Anim Dynamics - Bone: dyn_drawstring_03_l(ComponentPose=@n59, Alpha=1, ExternalForce="0,0,0") {
    angularDampingOverride=1
    angularSpringConstant=100
    bAngularSpring=true
    bGravityOverrideInSimSpace=true
    bOverrideAngularDamping=true
    bUseGravityOverride=true
    bUsePlanarLimit=false
    gravityScale=0.3
    numSolverIterationsPostUpdate=4
    numSolverIterationsPreUpdate=16
    alphaBoolBlend {blendOption=Linear}
    boundBone {boneName=dyn_drawstring_03_l}
    chainEnd {boneName=dyn_drawstring_03_l}
    gravityOverride {z=-980}
    physicsBodyDefinitions [{
      collisionType=CustomSphere
      sphereCollisionRadius=1
      boundBone {boneName=dyn_drawstring_03_l}
      boxExtents=<5 10 10>
      constraintSetup {angularLimitsMax=<0 60 60>, angularLimitsMin=<0 -60 0>, angularTarget=<1 0 0>}
      localJointOffset=<8 0 0>
     }]
    sphericalLimits [{limitRadius=20, drivingBone {boneName=thighFront_c}, sphereLocalOffset=<0 -20 0>}]
   }
   n61 = "Constraint - dyn_sidestring_01_Base (PPPPP)"!(
    ComponentPose=@n60
    Alpha=1
    ConstraintWeights_0=0.0912
    ConstraintWeights_1:float=0.0126
    ConstraintWeights_2:float=0.76
    ConstraintWeights_3:float=0.0249
    ConstraintWeights_4:float=0.1113
   ) -> (Pose) {
    alphaBoolBlend {blendOption=Linear}
    boneToModify {boneName=dyn_sidestring_01_Base}
    constraintSetup [
     {targetBone {boneName=spine_01}}
     {targetBone {boneName=spine_02}}
     {targetBone {boneName=dyn_pelvis_scale_r}}
     {targetBone {boneName=thigh_twist_01_r}}
     {targetBone {boneName=pelvis}}
    ]
    constraintWeights [1,1,1,1,1]
   }
   n62 = Anim Dynamics - Bone: dyn_sidestring1_01!(ComponentPose=@n61, ExternalForce="0,0,0") -> (Pose) {
    alpha=0.2
    angularDampingOverride=1
    angularSpringConstant=100
    bAngularSpring=true
    bGravityOverrideInSimSpace=true
    bOverrideAngularDamping=true
    bUseGravityOverride=true
    bUsePlanarLimit=false
    bUseSphericalLimits=true
    numSolverIterationsPostUpdate=4
    numSolverIterationsPreUpdate=16
    alphaBoolBlend {blendOption=Linear}
    boundBone {boneName=dyn_sidestring1_01}
    gravityOverride {z=-980}
    physicsBodyDefinitions [{
      sphereCollisionRadius=0
      boundBone {boneName=dyn_sidestring1_01}
      boxExtents=<20 10 15>
      constraintSetup {angularTargetAxis=AxisY, angularLimitsMax=<0 60 60>, angularLimitsMin=<0 -60 0>, angularTarget=<0 1 0>}
      localJointOffset=<2 0 0>
     }]
    sphericalLimits [
     {limitRadius=10, drivingBone {boneName=thigh_twist_01_r}, sphereLocalOffset=<-15 0 0>}
     {limitRadius=15, drivingBone {boneName=thigh_twist_01_r}, sphereLocalOffset=<-7.5 0 -5>}
    ]
   }
   n63 = Anim Dynamics - Bone: dyn_sidestring1_02!(ComponentPose=@n62, ExternalForce="0,0,0") -> (Pose) {
    alpha=0.4
    angularDampingOverride=1
    angularSpringConstant=100
    bAngularSpring=true
    bGravityOverrideInSimSpace=true
    bOverrideAngularDamping=true
    bUseGravityOverride=true
    bUsePlanarLimit=false
    bUseSphericalLimits=true
    numSolverIterationsPostUpdate=4
    numSolverIterationsPreUpdate=16
    alphaBoolBlend {blendOption=Linear}
    boundBone {boneName=dyn_sidestring1_02}
    gravityOverride {z=-980}
    physicsBodyDefinitions [{
      sphereCollisionRadius=0
      boundBone {boneName=dyn_sidestring1_02}
      boxExtents=<20 10 15>
      constraintSetup {angularTargetAxis=AxisY, angularLimitsMax=<0 60 60>, angularLimitsMin=<0 -60 0>, angularTarget=<0 1 0>}
      localJointOffset=<2 0 0>
     }]
    sphericalLimits [
     {limitRadius=10, drivingBone {boneName=thigh_twist_01_r}, sphereLocalOffset=<-15 0 0>}
     {limitRadius=17, drivingBone {boneName=thigh_twist_01_r}, sphereLocalOffset=<-10 0 -5>}
    ]
   }
   n64 = Anim Dynamics - Bone: dyn_sidestring1_03!(ComponentPose=@n63, ExternalForce="0,0,0") -> (Pose) {
    alpha=0.6
    angularDampingOverride=1
    angularSpringConstant=100
    bAngularSpring=true
    bGravityOverrideInSimSpace=true
    bOverrideAngularDamping=true
    bUseGravityOverride=true
    bUsePlanarLimit=false
    bUseSphericalLimits=true
    numSolverIterationsPostUpdate=4
    numSolverIterationsPreUpdate=16
    alphaBoolBlend {blendOption=Linear}
    boundBone {boneName=dyn_sidestring1_03}
    gravityOverride {z=-980}
    physicsBodyDefinitions [{
      sphereCollisionRadius=0
      boundBone {boneName=dyn_sidestring1_03}
      boxExtents=<20 10 15>
      constraintSetup {angularTargetAxis=AxisY, angularLimitsMax=<0 60 60>, angularLimitsMin=<0 -60 0>, angularTarget=<0 1 0>}
      localJointOffset=<2 0 0>
     }]
    sphericalLimits [
     {limitRadius=10, drivingBone {boneName=thigh_twist_01_r}, sphereLocalOffset=<-15 0 0>}
     {limitRadius=17, drivingBone {boneName=thigh_twist_01_r}, sphereLocalOffset=<-10 0 -5>}
    ]
   }
   n65 = Anim Dynamics - Bone: dyn_sidestring2_01!(ComponentPose=@n64, ExternalForce="0,0,0") -> (Pose) {
    alpha=0.6
    angularDampingOverride=1
    angularSpringConstant=100
    bAngularSpring=true
    bGravityOverrideInSimSpace=true
    bOverrideAngularDamping=true
    bUseGravityOverride=true
    bUsePlanarLimit=false
    bUseSphericalLimits=true
    numSolverIterationsPostUpdate=4
    numSolverIterationsPreUpdate=16
    alphaBoolBlend {blendOption=Linear}
    boundBone {boneName=dyn_sidestring2_01}
    gravityOverride {z=-980}
    physicsBodyDefinitions [{
      sphereCollisionRadius=0
      boundBone {boneName=dyn_sidestring2_01}
      boxExtents=<20 10 15>
      constraintSetup {angularTargetAxis=AxisY, angularLimitsMax=<0 60 60>, angularLimitsMin=<0 -60 0>, angularTarget=<0 1 0>}
      localJointOffset=<2 0 0>
     }]
    sphericalLimits [
     {limitRadius=10, drivingBone {boneName=thigh_twist_01_r}, sphereLocalOffset=<-15 0 0>}
     {limitRadius=17, drivingBone {boneName=thigh_twist_01_r}, sphereLocalOffset=<-10 0 -5>}
    ]
   }
   n66 = Anim Dynamics - Bone: dyn_sidestring3_01!(ComponentPose=@n65, ExternalForce="0,0,0") -> (Pose) {
    alpha=0.6
    angularBiasOverride=0.5
    angularDampingOverride=1
    angularSpringConstant=100
    bAngularSpring=true
    bGravityOverrideInSimSpace=true
    bOverrideAngularDamping=true
    bUseGravityOverride=true
    bUsePlanarLimit=false
    bUseSphericalLimits=true
    linearDampingOverride=0.9
    numSolverIterationsPostUpdate=4
    numSolverIterationsPreUpdate=16
    simulationSpace=BoneRelative
    alphaBoolBlend {blendOption=Linear}
    boundBone {boneName=dyn_sidestring3_01}
    gravityOverride {z=-980}
    physicsBodyDefinitions [{
      sphereCollisionRadius=0
      boundBone {boneName=dyn_sidestring3_01}
      boxExtents=<20 5 10>
      constraintSetup {angularTargetAxis=AxisY, angularLimitsMax=<0 0 60>, angularLimitsMin=<0 0 0>, angularTarget=<0 1 0>}
      localJointOffset=<0 0 0>
     }]
    relativeSpaceBone {boneName=pelvis}
    sphericalLimits [
     {limitRadius=10, drivingBone {boneName=thigh_twist_01_r}, sphereLocalOffset=<-15 0 0>}
     {limitRadius=17, drivingBone {boneName=thigh_twist_01_r}, sphereLocalOffset=<-10 0 -5>}
    ]
   }
   n67 = Anim Dynamics - Bone: dyn_sidestring4_01!(ComponentPose=@n66, ExternalForce="0,0,0") -> (Pose) {
    alpha=0.6
    angularDampingOverride=1
    angularSpringConstant=100
    bAngularSpring=true
    bGravityOverrideInSimSpace=true
    bOverrideAngularDamping=true
    bUseGravityOverride=true
    bUsePlanarLimit=false
    bUseSphericalLimits=true
    numSolverIterationsPostUpdate=4
    numSolverIterationsPreUpdate=16
    simulationSpace=BoneRelative
    alphaBoolBlend {blendOption=Linear}
    boundBone {boneName=dyn_sidestring4_01}
    gravityOverride {z=-980}
    physicsBodyDefinitions [{
      sphereCollisionRadius=0
      boundBone {boneName=dyn_sidestring4_01}
      boxExtents=<20 10 15>
      constraintSetup {angularTargetAxis=AxisY, angularLimitsMax=<0 0 60>, angularLimitsMin=<0 0 0>, angularTarget=<0 1 0>}
      localJointOffset=<2 0 0>
     }]
    relativeSpaceBone {boneName=pelvis}
    sphericalLimits [
     {limitRadius=10, drivingBone {boneName=thigh_twist_01_r}, sphereLocalOffset=<-15 0 0>}
     {limitRadius=17, drivingBone {boneName=thigh_twist_01_r}, sphereLocalOffset=<-10 0 -5>}
    ]
   }
   n68 = "Constraint - dyn_mask_01_Base (PPPPP)"!(
    ComponentPose=@n67
    Alpha=1
    ConstraintWeights_0=0.0149
    ConstraintWeights_1:float=0.1488
    ConstraintWeights_2:float=0.0871
    ConstraintWeights_3:float=0.3237
    ConstraintWeights_4:float=0.4255
   ) -> (Pose) {
    alphaBoolBlend {blendOption=Linear}
    boneToModify {boneName=dyn_mask_01_Base}
    constraintSetup [
     {targetBone {boneName=spine_01}}
     {targetBone {boneName=thigh_twist_01_l}}
     {targetBone {boneName=thigh_l}}
     {targetBone {boneName=dyn_hipPelvis_l}}
     {targetBone {boneName=pelvis}}
    ]
    constraintWeights [1,1,1,1,1]
   }
   n69 = Get skydiving_accel_scale() -> (skydiving_accel_scale:Vector)
   n70 = Anim Dynamics - Bone: dyn_mask_01_Base!(ComponentPose=@n68, ComponentLinearAccScale:Vector=@n69, ComponentLinearVelScale:Vector=@n69, ExternalForce="0,0,0") -> (Pose) {
    alpha=0.35
    angularDampingOverride=0.999
    angularSpringConstant=200
    bAngularSpring=true
    bGravityOverrideInSimSpace=true
    bOverrideAngularDamping=true
    bUsePlanarLimit=false
    gravityScale=0.3
    numSolverIterationsPostUpdate=4
    numSolverIterationsPreUpdate=16
    simulationSpace=RootRelative
    alphaBoolBlend {blendOption=Linear}
    boundBone {boneName=dyn_mask_01_Base}
    componentLinearAccScale=<1 1 0.2>
    componentLinearVelScale=<1 1 0.2>
    physicsBodyDefinitions [{
      sphereCollisionRadius=0
      boundBone {boneName=dyn_mask_01_Base}
      boxExtents=<10 10 10>
      constraintSetup {angularTargetAxis=AxisY, angularLimitsMax=<60 30 8>, angularLimitsMin=<-60 -30 -8>, angularTarget=<0 1 0>}
      localJointOffset=<-0.643 0.253 -5>
     }]
   }
   n71 = "Transform (Modify) Bone - Bone: dyn_collarBack"(ComponentPose=@n70, Alpha=1, Translation="1,0,0", Rotation="0,0,0", Scale="1,1,1") {
    rotationMode=BMM_Additive
    rotationSpace=BCS_BoneSpace
    scaleMode=BMM_Additive
    scaleSpace=BCS_BoneSpace
    translationMode=BMM_Additive
    translationSpace=BCS_BoneSpace
    alphaBoolBlend {blendOption=Linear}
    alphaScaleBiasClamp {interpSpeedDecreasing=25, interpSpeedIncreasing=25}
    boneToModify {boneName=dyn_collarBack}
    translation {x=1}
   }
   Output Pose(Result=Component To Local(ComponentPose=@n71)) {layerGroup=None}
  ]}
 EventGraph {nodes [
   Event BlueprintUpdateAnimation()
   n1 = GetOwnerSkeletalMeshComponent()
   n2 = BreakRotator(InRot=BreakTransform(InTransform=GetDeltaTransformFromRefPose(self=@n1, BoneName=calf_r, BaseName=thigh_r)).Rotation)
   > n3 = Set calf_rot_back_r(calf_rot_back_r:double=MapRangeClamped(Value=@n2.Yaw, InRangeA=15, InRangeB=115)) -> (Output_Get:double)
   > Set calf_rot_backExtreme_r(calf_rot_backExtreme_r:double=MapRangeClamped(Value=@n2.Yaw, InRangeA=115, InRangeB=135)) -> (Output_Get:double)
   n4 = GetOwnerSkeletalMeshComponent()
   n5 = BreakRotator(InRot=BreakTransform(InTransform=GetDeltaTransformFromRefPose(self=@n4, BoneName=calf_l, BaseName=thigh_l)).Rotation)
   > Set calf_rot_back_l(calf_rot_back_l:double=MapRangeClamped(Value=@n5.Yaw, InRangeA=15, InRangeB=115)) -> (Output_Get:double)
   > n6 = Set calf_rot_backExtreme_l(calf_rot_backExtreme_l:double=MapRangeClamped(Value=@n5.Yaw, InRangeA=115, InRangeB=125)) -> (Output_Get:double)
   > Set thigh_rot_both(thigh_rot_both:double=float * float(A=@n6, B=@n3)) -> (Output_Get:double)
   n7 = GetOwnerSkeletalMeshComponent()
   n8 = BreakRotator(InRot=BreakTransform(InTransform=GetDeltaTransformFromRefPose(self=@n7, BoneName=thigh_r, BaseName=pelvis)).Rotation)
   n9 = MapRangeClamped(Value=@n8.Yaw, InRangeA=0, InRangeB=-90)
   > Set thigh_rot_r(thigh_rot_r=@n9)
   > Set thigh_toSide_r(thigh_toSide_r:double=MapRangeClamped(Value=@n8.Pitch, InRangeA=0, InRangeB=45)) -> (Output_Get:double)
   n10 = GetDeltaTransformFromRefPose(self=GetOwnerSkeletalMeshComponent(), BoneName=thigh_l, BaseName=pelvis)
   n11 = MapRangeClamped(Value=BreakRotator(InRot=BreakTransform(InTransform=@n10).Rotation).Yaw, InRangeA=-15, InRangeB=-110)
   > Set thigh_rot_l(thigh_rot_l=@n11)
   > Set thigh_rot_either(thigh_rot_either:double="Max (Float)"(A=@n11, B=@n9)) -> (Output_Get:double)
   > Set skydiving_accel_scale(skydiving_accel_scale:Vector=SelectVector(A="1,1,0.2", B="1,1,1", bPickA=Get bIsSkyDiving())) -> (Output_Get:Vector)
   > Set jumping_alpha(jumping_alpha:double="To Float (Boolean)"(InBool=Get bIsJumping())) -> (Output_Get:double)
   > Set sprinting_alpha(sprinting_alpha:double="To Float (Boolean)"(InBool=Get bIsSprinting())) -> (Output_Get:double)
   TryGetPawnOwner()
  ]}
]
variables [
 calf_rot_back_l {type=number}
 calf_rot_back_r {type=number}
 calf_rot_backExtreme_l {type=number}
 calf_rot_backExtreme_r {type=number}
 jumping_alpha {type=number}
 Shorts_Alpha {type=number, default=0.7}
 skydiving_accel_scale {type=Vector}
 sprinting_alpha {type=number}
 thigh_rot_both {type=number}
 thigh_rot_either {type=number}
 thigh_rot_l {type=number, default=1}
 thigh_rot_r {type=number, default=1}
 thigh_toSide_r {type=number, default=1}
]
