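# Physics asset for the face accessory mesh of the Drift Summer medium male player cosmetic, defining physical bodies for its hair bones.
# Eight hair bones carry thin capsules of varying length and tilt, alongside one wide capsule on the head bone.
# Each hair body is constrained directly to the head rather than chained, with identical narrow swing limits and very tight twist limits that restrict movement to a slight sway.
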PhysicsAsset /BRCosmetics/Characters/Player/Male/Medium/Bodies/M_MED_Drift_Summer/Meshes/Parts/M_MED_Drift_Summer_FaceAcc_Physics
bodies {
 hair_0 [{shapeType=Capsule, length=7, radius=1, center=<-0.2961 1.1316 5.1376>}]
 hair_1 [{shapeType=Capsule, length=6, radius=1, center=<0.0001 1.0504 3.4251>, rotation=<rot 0 0 10.0008>}]
 hair_2 [{shapeType=Capsule, length=6.5, radius=1, center=<0.0001 1.1352 2.1056>, rotation=<rot 0 0 20.0001>}]
 hair_3 [{shapeType=Capsule, length=6.25, radius=1, center=<1.0348 0.8841 2.6403>, rotation=<rot -10 0 20.0001>}]
 hair_4 [{shapeType=Capsule, length=4.25, radius=1, center=<0 -0.5417 2.114>, rotation=<rot 0 0 -20.0015>}]
 hair_5 [{shapeType=Capsule, length=2, radius=1, center=<0 -0.7366 1.0681>, rotation=<rot 0 0 -29.9994>}]
 hair_6 [{shapeType=Capsule, length=2, radius=1, center=<0.4911 -0.8929 0.5108>, rotation=<rot 0 0 -20.0001>}]
 hair_7 [{shapeType=Capsule, length=2, radius=1, center=<0 0 -1>, rotation=<rot 0 0 -39.9998>}]
 head [{shapeType=Capsule, length=5, radius=5}]
}
constraints [
 {bone1Name=hair_0, bone2Name=head, swing1LimitDegrees=20, swing2LimitDegrees=20, twistLimitDegrees=2}
 {bone1Name=hair_1, bone2Name=head, swing1LimitDegrees=20, swing2LimitDegrees=20, twistLimitDegrees=2}
 {bone1Name=hair_2, bone2Name=head, swing1LimitDegrees=20, swing2LimitDegrees=20, twistLimitDegrees=2}
 {bone1Name=hair_3, bone2Name=head, swing1LimitDegrees=20, swing2LimitDegrees=20, twistLimitDegrees=2}
 {bone1Name=hair_4, bone2Name=head, swing1LimitDegrees=20, swing2LimitDegrees=20, twistLimitDegrees=2}
 {bone1Name=hair_5, bone2Name=head, swing1LimitDegrees=20, swing2LimitDegrees=20, twistLimitDegrees=2}
 {bone1Name=hair_6, bone2Name=head, swing1LimitDegrees=20, swing2LimitDegrees=20, twistLimitDegrees=2}
 {bone1Name=hair_7, bone2Name=head, swing1LimitDegrees=20, swing2LimitDegrees=20, twistLimitDegrees=2}
]
